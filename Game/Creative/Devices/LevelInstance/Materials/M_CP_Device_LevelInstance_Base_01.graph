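# Translucent unlit hologram material for the Creative Level Instance device base, with an IsHologram switch that swaps in striped two-color shading.
# Flickering world-aligned noise and a TurnOn-driven height pulse modulate Fresnel and displacement.
# Fresnel, actor-anchored screen-space scanlines and a vertical gradient index a curve atlas, combined with texture overlay, reflection, sun specular and outline glow.
# Opacity combines scanlines, noise, depth fade and a custom-depth occlusion test that mobile skips.
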Material /Game/Creative/Devices/LevelInstance/Materials/M_CP_Device_LevelInstance_Base_01
graph {
 properties {
  allowTranslucentCustomDepthWrites=true
  bEnableResponsiveAA=true
  bUsedWithMeshDeformer=true
  bUsedWithNiagaraMeshParticles=true
  bUsedWithNiagaraSprites=true
  bUsedWithSkeletalMesh=true
  blendMode=BLEND_Translucent
  numCustomizedUVs=4
  opacityMaskClipValue=0
  pixelDepthOffsetMode=PDOM_Legacy
  refractionMethod=RM_IndexOfRefraction
  shadingModel=MSM_Unlit
  usedShadingModels=Unlit
 }
 nodes [
  Constant() {r=0.95}
  Constant() {r=1}
  n1 = ComponentMask(None=WorldPosition().XYZ) {g=true, r=true}
  n2 = Multiply(A=ScalarParameter() {defaultValue=12, group=Color, parameterName=Color Mask Tiling}) {constB=0.001}
  n3 = LinearInterpolate(
   A=VectorParameter() {group=Color, parameterName=Color 1, defaultValue=<rgba 0 0 0 1>}.RGB
   B=VectorParameter() {group=Color, parameterName=Color 2, defaultValue=<rgba 1 1 1 1>}.RGB
   Alpha=TextureSampleParameter2D(UVs=Multiply(A=@n1, B=@n2)) {group=Color, parameterName=Color Mask, texture=/Game/Creative/Devices/Generic/Materials/Textures/T_CP_Device_Stripes_01_D}.RGB
  )
  Multiply(A=@n3)
  n4 = TextureObject() {texture=/Game/Global/Textures/Default/Noise/Noise}
  n5 = Constant() {r=32}
  n6 = Multiply(
   A=LinearInterpolate(
    A=MaterialFunctionCall(TextureObject=@n4, TextureSize=@n5) {materialFunction=/Game/Weapons/WeaponSkins/Wraps/MaterialFunctions/WorldAlignedTexture}.XYZ Texture
    Alpha=ScalarParameter() {defaultValue=0.9, group=Noise, parameterName=Detail Noise Reduction}
   ) {constB=0.5}
   B=ScalarParameter() {defaultValue=0.6, group=Noise, parameterName=Detail Noise Brightness}
  )
  n7 = LinearInterpolate(
   A=MaterialFunctionCall(
    TextureObject=TextureObject() {texture=/Game/Creative/Devices/Generic/Materials/Textures/T_CP_Device_FortNoise_01_D}
    TextureSize=ScalarParameter() {defaultValue=256, group=Noise, parameterName=Noise Texture Size}
   ) {materialFunction=&WorldAlignedTexture}.XYZ Texture
   Alpha=ScalarParameter() {defaultValue=0.3, group=Noise, parameterName=Noise Contrast Reduction}
  ) {constB=0.5}
  n8 = Multiply(A=@n7, B=ScalarParameter() {defaultValue=0.5, group=Noise, parameterName=Noise Brightness})
  Reroute(None=Saturate(None=Add(A=@n6, B=@n8)))
  Saturate(None=@n7)
  ScalarParameter() {defaultValue=0.5, parameterName=Metallic}
  n9 = WorldPosition()
  n10 = ObjectPositionWS()
  n11 = Multiply(A=Normalize(VectorInput=Subtract(A=@n9.XYZ, B=@n10)), B=Constant3Vector() {constant=<rgba 1 1 0 0>})
  n12 = Add(
   A=ScalarParameter() {defaultValue=10, group=Noise, parameterName=WPOStrength}
   B=Multiply(A=ObjectRadius(), B=ScalarParameter() {defaultValue=0.025, group=Noise, parameterName=WPOStrengthFromObjectRadius}) {constB=0.05}
  )
  n13 = Divide(A=ObjectRadius(), B=ScalarParameter() {defaultValue=0.1, group=Noise, parameterName=NoiseSize})
  n14 = Multiply(A=Time(), B=VectorParameter() {group=Noise, parameterName=NoisePanSpeed, defaultValue=<rgba -0.08 0.07 -0.2 0>}.RGB)
  n15 = MaterialFunctionCall(
   TextureObject=TextureObject() {samplerType=SAMPLERTYPE_LinearColor, texture=&T_NoisePack_LinearColor}
   TextureSize=Multiply(A=VectorParameter() {group=Noise, parameterName=NoiseTileMult, defaultValue=<rgba 1 1 0.1 1>}.RGB, B=@n13)
   WorldPosition=Add(A=WorldPosition().XYZ, B=Multiply(A=@n13, B=@n14))
  ) {materialFunction=/Engine/Functions/Engine_MaterialFunctions01/Texturing/WorldAlignedTexture}
  n16 = ActorPositionWS()
  n17 = MaterialFunctionCall(Float2=Round(None=Divide(A=@n16) {constB=2000})) {materialFunction=&BreakOutFloat2Components}
  n18 = Add(A=Time(), B=Add(A=@n17.R, B=@n17.R))
  n19 = Multiply(A=ScalarParameter() {defaultValue=0.5, parameterName=FlickerSpeed}, B=@n18)
  n20 = AppendVector(A=Sine(None=@n19) {period=1.6}, B=Sine(None=@n19) {period=1.42})
  n21 = Sine(None=@n19) {period=1.51}
  n22 = Sine(None=@n19) {period=2.1}
  n23 = ScalarParameter() {defaultValue=0.5, group=Noise, parameterName=NoiseFrequency}
  n24 = Multiply(B=@n23) {constA=0.25}
  n25 = Add(A=Sine(None=@n19) {period=0.72}, B=Multiply(B=@n23) {constA=0.75}) {constB=0.75}
  n26 = Add(A=Multiply(B=@n23) {constA=0.9, constB=0.9}, B=Sine(None=@n19) {period=1.25}) {constB=0.9}
  n27 = Multiply(A=Multiply(A=Add(A=@n22, B=@n24), B=@n25), B=Multiply(A=@n25, B=Ceil(None=@n26)))
  n28 = ConstantBiasScale(None=Sine(None=@n18) {period=0.075})
  n29 = Multiply(A=Ceil(None=AppendVector(A=@n20, B=@n21)), B=Multiply(A=Saturate(None=@n27), B=LinearInterpolate(Alpha=@n28) {constA=0.7}))
  n30 = ScalarParameter() {defaultValue=2, group=Noise, parameterName=NoiseMult}
  n31 = Multiply(
   A=Multiply(A=Power(Base=@n15.XY Texture) {constExponent=4}, B=ScalarParameter() {defaultValue=2, group=Noise, parameterName=NoiseMaskMult}) {constB=20}
   B=Add(A=ScalarParameter() {group=Noise, parameterName=NoiseAdd}, B=Multiply(A=@n29, B=@n30))
  )
  n32 = WorldPosition()
  n33 = ComponentMask(None=MaterialFunctionCall() {materialFunction=&ObjectPivotPoint}.Object Pivot Location) {b=true}
  n34 = Add(A=@n33, B=ScalarParameter() {defaultValue=300, parameterName=PulseHeight}) {constB=1600}
  n35 = ScalarParameter() {defaultValue=1, parameterName=TurnOn}
  n36 = ScalarParameter() {defaultValue=33, parameterName=PulseRadius}
  n37 = Multiply(A=@n36) {constB=0.1}
  n38 = Multiply(
   A=SphereMask(
    A=ComponentMask(None=@n32.XYZ) {b=true}
    B=LinearInterpolate(A=@n33, B=@n34, Alpha=@n35) {constA=-0.25, constB=1.2}
    Radius=LinearInterpolate(A=@n36, B=@n37, Alpha=@n35) {constA=500, constB=100}
   ) {hardnessPercent=0}
   B=OneMinus(None=@n35)
  )
  n39 = Add(A=Desaturation(None=Saturate(None=@n31)), B=Multiply(A=@n38) {constB=2})
  StaticSwitchParameter(True=Multiply(A=@n39, B=Multiply(A=@n11, B=@n12)) {constB=0}, False=Constant()) {defaultValue=true, parameterName=IsHologram, sortPriority=0}
  n40 = ComponentMask(None=Transform(None=ReflectionVectorWS()) {transformSourceType=TRANSFORMSOURCE_World, transformType=TRANSFORM_View}) {g=true, r=true}
  n41 = ScalarParameter() {defaultValue=1, group=Reflection, parameterName=ReflectionMapTilingU, sortPriority=5}
  n42 = ScalarParameter() {defaultValue=1, group=Reflection, parameterName=ReflectionMapTilingV, sortPriority=5}
  n43 = TextureSampleParameter2D(UVs=Multiply(A=ConstantBiasScale(None=@n40), B=AppendVector(A=@n41, B=@n42))) {group=Reflection, parameterName=SphereReflectionMap, sortPriority=1, texture=/Game/Packages/Fortress_SharedMaterials/Textures/ReflectionMap2}
  n44 = Multiply(
   A=Power(Base=@n43.RGB, Exp=ScalarParameter() {defaultValue=1, group=Reflection, parameterName=ReflectionExponent})
   B=VectorParameter() {group=Reflection, parameterName=Reflection Tint, sortPriority=2, defaultValue=<rgba 1 1 1 1>}.RGB
  )
  n45 = ScalarParameter() {defaultValue=1, group=Reflection, parameterName=Reflection Multiplier}
  n46 = CollectionParameter() {
   collection=&FortniteMaterialParameters
   parameterName=SunAndMoonModelDirectionalVector
  }
  n47 = ComponentMask(None=@n46) {b=true, g=true, r=true}
  n48 = ReflectionVectorWS()
  n49 = Power(Base=Clamp(None=DotProduct(A=@n47, B=@n48)), Exp=ScalarParameter() {defaultValue=200, group=Specular, parameterName=Spec Highlight Power}) {constExponent=20}
  n50 = ScalarParameter() {defaultValue=2, group=Specular, parameterName=Spec Intense}
  n51 = VertexNormalWS()
  n52 = Transform(None=TextureSampleParameter2D() {group=Texture, parameterName=Normal, samplerType=SAMPLERTYPE_Normal, texture=/Engine/EngineMaterials/BaseFlattenNormalMap}.RGB)
  n53 = ScalarParameter() {group=Lighting, parameterName=UseNormal}
  n54 = Clamp(None=DotProduct(A=VectorParameter() {parameterName=LightDirection, defaultValue=<rgba 0 0 1 1>}.RGB, B=LinearInterpolate(A=@n51, B=@n52, Alpha=@n53)))
  n55 = ScalarParameter() {defaultValue=0.2, group=Lighting, parameterName=NormalBrightnessAdd}
  n56 = QualitySwitch(
   Default=MaterialFunctionCall(In=@n39, Contrast=ScalarParameter() {defaultValue=0.2, group=Noise, parameterName=NoiseContrast}) {materialFunction=&CheapContrast}.Result
   Low=Constant()
  )
  n57 = LinearInterpolate(B=ScalarParameter() {defaultValue=0.6, group=Noise, parameterName=NoiseFresnelXPMult}, Alpha=@n56) {constA=1, constB=0.5}
  n58 = PixelNormalWS()
  n59 = Fresnel(
   ExponentIn=Multiply(A=ScalarParameter() {defaultValue=1, parameterName=FresnelExp}, B=@n57)
   BaseReflectFractionIn=ScalarParameter() {defaultValue=0.5, parameterName=FresnelMin}
   Normal=@n58
  ) {baseReflectFraction=0, exponent=0.5}
  n60 = Multiply(A=@n57, B=ScalarParameter() {defaultValue=1, parameterName=EdgeFresnelXP})
  n61 = Saturate(None=LinearInterpolate(Alpha=Fresnel(ExponentIn=@n60, Normal=@n58) {baseReflectFraction=0, exponent=1}) {constA=5, constB=-0.4})
  n62 = Saturate(None=@n56)
  n63 = Add(
   A=Power(Base=Multiply(A=@n59, B=@n61))
   B=Multiply(A=ScalarParameter() {defaultValue=-0.4, group=Noise, parameterName=NoiseBrightness}, B=@n62) {constB=0.1}
  )
  n64 = Multiply(A=Constant() {r=60}, B=LinearInterpolate(Alpha=Ceil(None=@n62)) {constA=1, constB=1.25})
  n65 = ScreenPosition()
  n66 = MaterialFunctionCall(Float2=ViewSize()) {materialFunction=&BreakOutFloat2Components}
  n67 = AppendVector(A=Constant() {r=1}, B=Divide(A=@n66.G, B=@n66.G))
  n68 = ConstantBiasScale(None=Multiply(A=ConstantBiasScale(None=@n65.ViewportUV) {bias=-0.5, scale=2}, B=@n67) {desc=Keep 1:1 aspect ratio regardless of screen aspect ratio}) {scale=2}
  n69 = Constant2Vector() {g=-1, r=1}
  n70 = MaterialFunctionCall(World Position=ActorPositionWS() {desc=Also use Object Position}) {materialFunction=&TransformToClipSpace}
  n71 = Multiply(A=Multiply(A=@n67, B=Multiply(A=@n69, B=@n70.Clip Space XY))) {constB=0.5, desc=Halve offsets to account for 0...1 UV space}
  n72 = Add(
   A=MaterialFunctionCall(UVs=Subtract(A=@n68, B=@n71), "Rotation Angle (0-1)"=Constant() {r=0.25}) {materialFunction=&CustomRotator}.Rotated Values
   B=Multiply(A=Time(), B=Constant() {r=-0.02})
  )
  n73 = ConstantBiasScale(None=Sine(None=Multiply(A=@n64, B=@n72) {constA=1}))
  n74 = Saturate(None=Add(A=Constant() {r=0.5}, B=@n73))
  n75 = ComponentMask(None=Subtract(A=WorldPosition().XYZ, B=ObjectPositionWS())) {b=true}
  n76 = ObjectRadius()
  n77 = Saturate(None=ConstantBiasScale(None=Divide(A=@n75, B=@n76)))
  n78 = OneMinus(None=@n77)
  n79 = ScalarParameter() {parameterName=FlipGradient}
  n80 = Add(
   A=LinearInterpolate(
    A=Multiply(A=@n63, B=@n74)
    B=LinearInterpolate(A=@n77, B=@n78, Alpha=@n79)
    Alpha=ScalarParameter() {defaultValue=0.4, parameterName=VerticalGradientInflience}
   )
   B=ScalarParameter() {parameterName=Fill}
  )
  n81 = Add(A=Multiply(A=ScalarParameter() {defaultValue=1, parameterName=TurnOn}, B=Clamp(None=@n80)), B=@n38) {constA=1}
  n82 = Saturate(None=Add(A=Multiply(A=@n54, B=@n55) {constB=0.15}, B=@n81))
  n83 = CurveAtlasRowParameter(CurveTime=MaterialFunctionCall(Float2=@n82) {materialFunction=&BreakOutFloat2Components}.R) {
   atlas=&AthenaCurveAtlas01
   curve=/Game/Creative/Devices/Generic/LinearColorCurves/Curve_CP_Device_Hologram_01
   defaultValue=282
   desc=Curve Atlas
   group=Color
   parameterName=CurveAtlas
  }
  n84 = TextureSampleParameter2D() {
   group=Diffuse Texture Overlay
   parameterName=Diffuse Texture Overlay
   texture=/Game/Creative/Devices/LevelInstance/Textures/T_CP_Device_LevelInstance_Base_Hologram_01_D
  }
  n85 = LinearInterpolate(
   A=@n84.RGB
   B=MaterialFunctionCall(In=@n84.RGB, Contrast=ScalarParameter() {group=Diffuse Texture Overlay, parameterName=Diffuse Contribution Cheap Contrast}) {materialFunction=&CheapContrast}.Result
   Alpha=ScalarParameter() {group=Diffuse Texture Overlay, parameterName=Diffuse Overlay Use Cheap Contrast}
  )
  n86 = ScalarParameter() {defaultValue=1, group=Diffuse Texture Overlay, parameterName=Diffuse Contribution Desaturation}
  n87 = LinearInterpolate(
   A=@n83
   B=Multiply(A=Desaturation(None=@n85, Fraction=@n86), B=@n83)
   Alpha=ScalarParameter() {group=Diffuse Texture Overlay, parameterName=Lerp Diffuse Overlay}
  )
  n88 = StaticSwitchParameter(True=@n83, False=@n82) {defaultValue=true, group=Color, parameterName=UseColorCurve, sortPriority=0}
  n89 = Add(
   A=Add(A=Multiply(A=@n44, B=@n45), B=Multiply(A=@n49, B=@n50))
   B=Multiply(
    A=StaticSwitchParameter(True=@n87, False=@n88) {defaultValue=true, group=Diffuse Texture Overlay, parameterName=UseDiffuseTextureOverlay, sortPriority=0}
    B=ScalarParameter() {defaultValue=2, parameterName=EmissiveMult}
   ) {constB=2}
  )
  n90 = ScalarParameter() {group=Outline, parameterName=Inside Darken}
  n91 = TextureSampleParameter2D() {
   group=Diffuse Texture Overlay
   parameterName=Diffuse Texture Overlay
   texture=&T_CP_Device_LevelInstance_Base_Hologram_01_D
  }
  n92 = Power(
   Base=StaticSwitchParameter(True=@n91.RGB, False=@n89) {defaultValue=true, group=Diffuse Texture Overlay, parameterName=UseDiffuseTextureOverlay, sortPriority=0}
   Exp=ScalarParameter() {defaultValue=4, group=Outline, parameterName=Outline Exponent}
  )
  n93 = Add(
   A=Multiply(A=@n89, B=Subtract(B=@n90))
   B=LinearInterpolate(
    A=Constant()
    B=ScalarParameter() {defaultValue=4, group=Outline, parameterName=Outline Glow Intensity}
    Alpha=MaterialFunctionCall(Float3=@n92) {materialFunction=&BreakOutFloat3Components}.R
   )
  )
  StaticSwitchParameter(True=@n93, False=@n3) {defaultValue=true, parameterName=IsHologram, sortPriority=0}
  n94 = ParticleColor()
  n95 = ScalarParameter() {parameterName=UseParticleColor}
  n96 = DepthFade(Opacity=@n81, FadeDistance=ScalarParameter() {defaultValue=5, parameterName=FadeDistance})
  n97 = Constant() {r=0.6}
  n98 = Multiply(A=@n56) {constB=1.5}
  n99 = LinearInterpolate(
   A=Clamp(None=Subtract(A=@n97, B=@n98)) {minDefault=-0.25}
   B=Add(A=@n98)
   Alpha=MaterialFunctionCall(Float2=Power(Base=@n73) {constExponent=1.5}) {materialFunction=&BreakOutFloat2Components}.R
  ) {constA=0.5}
  n100 = Multiply(A=Saturate(None=Multiply(A=@n96, B=@n99)), B=ScalarParameter() {defaultValue=1.25, parameterName=OpacityMult})
  n101 = PixelDepth()
  n102 = ScalarParameter() {defaultValue=-0.25, parameterName=BiasSorting_Far}
  n103 = SceneTexture() {sceneTextureId=PPI_CustomDepth}
  n104 = Constant()
  n105 = ShadingPathSwitch(
   Default=If(A=Add(A=@n101.R, B=@n102), B=ComponentMask(None=@n103.Color) {r=true}, "A > B"=@n104, "A == B"=@n104, "A < B"=Constant() {r=1})
   Mobile=Constant() {r=1}
  ) {desc=Migrated from FeatureLevelSwitch}
  StaticSwitchParameter(
   True=Multiply(A=LinearInterpolate(B=@n94.A, Alpha=@n95) {constA=1, desc=Workaround for skel mesh bug}, B=Multiply(A=@n100, B=@n105))
   False=Constant() {r=1}
  ) {defaultValue=true, parameterName=IsHologram, sortPriority=0}
  StaticSwitchParameter(True=@n93, False=Constant()) {defaultValue=true, parameterName=IsHologram, sortPriority=0}
  TextureSampleParameter2D() {
   group=Diffuse Texture Overlay
   parameterName=Diffuse Texture Overlay
   texture=&T_CP_Device_LevelInstance_Base_Hologram_01_D
  }
 ]
}
