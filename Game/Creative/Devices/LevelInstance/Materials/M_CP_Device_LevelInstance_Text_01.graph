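# Unlit, translucent, two-sided material for distance-field font text on the Creative Level Instance device.
# The font's red channel is steeply remapped to form a sharp glyph mask, multiplied by an optional, invertible camera-distance fade.
# A SphereMask on the same distance value blends between fill and outline colors.
# TwoSidedSign applies a separate brightness to back faces.
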Material /Game/Creative/Devices/LevelInstance/Materials/M_CP_Device_LevelInstance_Text_01
graph {
 properties {blendMode=BLEND_Translucent, pixelDepthOffsetMode=PDOM_Legacy, shadingModel=MSM_Unlit, twoSided=true, usedShadingModels=Unlit}
 nodes [
  n1 = FontSampleParameter() {font=/Engine/EngineFonts/RobotoDistanceField, parameterName=Font}
  n2 = LinearInterpolate(Alpha=@n1.R) {constA=-90, constB=90}
  n3 = Distance(A=CameraPositionWS(), B=WorldPosition().XYZ)
  n4 = ScalarParameter() {defaultValue=-100, group=Distance Fade, parameterName=DistanceBias}
  n5 = Saturate(None=Divide(A=Add(A=@n3, B=@n4), B=ScalarParameter() {defaultValue=1200, group=Distance Fade, parameterName=FadeScale}))
  n6 = OneMinus(None=@n5)
  n7 = ScalarParameter() {group=Distance Fade, parameterName=Invert Distance Fade, sortPriority=1}
  LinearInterpolate(
   A=@n2
   B=Multiply(A=@n2, B=LinearInterpolate(A=@n6, B=@n5, Alpha=@n7) {constA=1})
   Alpha=ScalarParameter() {defaultValue=1, group=Distance Fade, parameterName=Enable Distance Fade, sortPriority=1}
  ) {constA=1}
  n8 = LinearInterpolate(
   A=VectorParameter() {parameterName=Color, defaultValue=<rgba 1 1 1 1>}.RGB
   B=VectorParameter() {parameterName=Outline Color, defaultValue=<rgba 1 1 1 1>}.RGB
   Alpha=SphereMask(
    A=@n1.R
    B=ScalarParameter() {defaultValue=0.01, parameterName=OutlinePos}
    Radius=ScalarParameter() {defaultValue=0.5, parameterName=OutlineThickness}
    Hardness=ScalarParameter() {defaultValue=1, parameterName=OutlineHardness}
   )
  )
  n9 = ScalarParameter() {defaultValue=0.55, parameterName=Backface Brightness}
  Reroute(None=LinearInterpolate(A=Multiply(A=@n8, B=@n9) {constB=0.5}, B=@n8, Alpha=TwoSidedSign()))
 ]
}
